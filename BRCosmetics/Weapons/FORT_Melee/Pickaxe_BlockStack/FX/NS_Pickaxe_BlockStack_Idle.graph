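# Idle Niagara effect for the BlockStack pickaxe: a flickering orange-red glow with small rising cube embers.
# CapsuleGlow burst-spawns persistent sprite and mesh glows, their alpha driven by layered waveforms, system fade and front-end scaling.
# Cubes spawns mesh and sprite particles at baked socket offsets, pausing during swings.
# Drag, curl noise and upward acceleration move them as they scale and fade, with a grid-based CustomPos computed after solving.
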NiagaraSystem /BRCosmetics/Weapons/FORT_Melee/Pickaxe_BlockStack/FX/NS_Pickaxe_BlockStack_Idle
dependencies {
 usedDataInterfaces [
  &NiagaraDataInterfaceArrayFloat3
  &NiagaraDataInterfaceCurve
  &NiagaraDataInterfaceColorCurve
  &NiagaraDataInterfaceSocketReader
 ]
 usedDynamicInputs []
 usedModules [
  &self:SystemSpawnScript.SystemScriptSource.SystemScriptGraph.NiagaraNodeAssignment_0.SetVariables_#1
  &SystemState
  &self:SystemSpawnScript.SystemScriptSource.SystemScriptGraph.NiagaraNodeAssignment_1.SetVariables_#2
 ]
}
emitters {
 CapsuleGlow {
  emitterUpdateScript [
   &EmitterState(
    Inactive Response=::NewEnumerator1
    Life Cycle Mode=::NewEnumerator1
    Loop Behavior=::NewEnumerator1
    Loop Duration Mode=::NewEnumerator1
    UseLoopDelay=false
   )
   &SpawnBurst_Instantaneous(Use Spawn Probability=false, UseLoopCountLimit=false)
  ]
  particleSpawnScript [&InitializeParticle(
    Position Offset=<0 0 35>
    Sprite Rotation Mode=::NewEnumerator0
    Sprite Size=<25 55>
    Sprite Size Mode=::NewEnumerator4
    UsePositionOffset=true
    Write Material Random=true
   )]
  particleUpdateScript [
   &ParticleState(Kill Particles When Lifetime Has Elapsed=false, Let Infinitely Lived Particles Die When Emitter Deactivates=true)
   /SetVariables(Particles.FlickerAlpha:NiagaraFloat=&Waveform(
     Waveform Count=::NewEnumerator2
     Global Phase=&Multiply_Float(
      A=/Niagara/DynamicInputs/Integer/V2/Round_Float(Float:NiagaraFloat=&Divide_Float(A=$Emitter.Age, B=0.05))
      B=0.05
     )
     "Waveform [1]"=::NewEnumerator0
     "[1] Frequency"=1
     "[1] Amplitude Min/Max"=<0.5 1>
     "Waveform [2] Blend"=::NewEnumerator2
     "Waveform [2]"=::NewEnumerator2
     "[2] Input Position"=3.14
     "[2] Phase"=$Emitter.Age
     "[2] Phase Scale"=1
     "[2] Compound Frequency"=2
     "[2] Frequency"=2
     "[2] Amplitude Min/Max"=<0.35 1>
     "[2] Scale"=1
     "[2] Bias"=0
     "[2] Exponent"=1
     "[2] Pulse Duration"=0.1
     "[2] Clip Triangle Min/Max"=<0 0>
     "[2] Random Seed"=0
     "Waveform [3] Blend"=::NewEnumerator2
     "Waveform [3]"=::NewEnumerator5
     "[3] Input Position"=1.33
     "[3] Phase"=$Emitter.Age
     "[3] Phase Scale"=1
     "[3] Frequency"=0.5
     "[3] Compound Frequency"=2
     "[3] Amplitude Min/Max"=<0.7 1>
     "[3] Scale"=1
     "[3] Bias"=0
     "[3] Exponent"=1
     "[3] Pulse Duration"=0.1
     "[3] Clip Triangle Min/Max"=<0 0>
     "[3] Random Seed"=0
    )) {isSetParameters=true}
   &DynamicMaterialParameters(
    Index 0 Param 1=1
    Module.Index0Float0Write=true
    Module.Index0Float1Write=false
    Module.Index0Float2Write=false
    Module.Index0Float3Write=false
   )
   &ScaleColor(
    Scale Alpha=&Multiply_Float(
     A=$Particles.FlickerAlpha
     B=&Multiply_Float(A=&FN_FrontEndFloatModifier(Gameplay=0.03, Front End=0.0175), B=$System.SystemFade)
    )
    Scale RGB=&MakeVectorFromLinearColorRGB(Color=<rgba 15 1.5 0 1>)
    ScaleA=true
    ScaleRGB=true
   )
   &FN_Mesh_Attributes(
    Color=&MultiplyLinearColor(
     Linear Color=<rgba 15 1.5 0 0.11>
     Scale Factor=&MakeVector4FromVectorAndFloat(
      "Vector (RGB)"=<1 1 1>
      "Float (Alpha)"=&Multiply_Float(
       A=$Particles.FlickerAlpha
       B=&Multiply_Float(A=&FN_FrontEndFloatModifier(Gameplay=1, Front End=0.65), B=$System.SystemFade)
      )
     )
    )
    Create Color Attribute=true
    Create Position Attribute=true
    Module Output=::NewEnumerator1
    Position:NiagaraPosition=$Engine.Emitter.SimulationPosition
   )
   &FN_Sprite_Attributes(
    Color=&MultiplyLinearColor(
     Linear Color=<rgba 15 1.5 0 0.035>
     Scale Factor=&MakeVector4FromVectorAndFloat(
      "Vector (RGB)"=<1 1 1>
      "Float (Alpha)"=&Multiply_Float(
       A=$Particles.FlickerAlpha
       B=&Multiply_Float(A=&FN_FrontEndFloatModifier(Gameplay=1, Front End=0.65), B=$System.SystemFade)
      )
     )
    )
    Create Color Attribute=true
    Create Position Attribute=true
    Create SpriteSize Attribute=true
    Module Output:ENiagara_AttributeSamplingApplyOutput=::NewEnumerator1
    Position:NiagaraPosition=&AddVectorToPosition(Position=$Engine.Emitter.SimulationPosition, Vector=<0 0 32.5>)
    SpriteSize=&Vector2DFromFloat(Value=52.5)
   )
   &SpriteFacingAndAlignment(Sprite Alignment=<0 0 1>, Sprite Facing=<0 -1 0>, WriteAlignment=true, WriteFacing=true)
  ]
  rendererClasses [&NiagaraSpriteRendererProperties, &NiagaraMeshRendererProperties]
  renderers [
   {rendererClass=&NiagaraSpriteRendererProperties}
   {rendererClass=&NiagaraMeshRendererProperties, rendererIndex=1}
   {rendererClass=&NiagaraSpriteRendererProperties, rendererIndex=2}
  ]
 }
 Cubes {
  emitterUpdateScript [
   &EmitterState()
   &SpawnRate(
    Spawn Probability=0.66
    SpawnRate=&MakeCustomFloatFromBool(Bool=$User.bIsSwinging, True Float=0, False Float=15)
    Use Spawn Probability=true
   )
  ]
  particleSpawnScript [
   &InitializeParticle(
    Lifetime=&RandomRangeFloat(Minimum=0.35, Maximum=0.5)
    Mesh Scale Mode=::NewEnumerator1
    Mesh Uniform Scale Max=1.25
    Mesh Uniform Scale Min=0.7
    Sprite Size Mode=::NewEnumerator3
    Uniform Sprite Size=9
    UsePositionOffset=false
    Write Material Random=true
   )
   &SocketLocation(
    SamplePosOffset=true
    Sampled Position Offset=&SelectVectorFromArray(
     Array Sampling Mode=::NewEnumerator0
     Direct Array Index=?
     Vector Selection Array=DataInterface() {propertyValues {FloatData [
        <0.5234 -15.1629 37.3227>
        <0.3635 -14.0633 38.2803>
        <0.1307 -12.6623 39.5003>
        <-1.6191 -13.292 38.952>
        <-1.6845 8.5721 41.0288>
        <-0.9783 8.3204 41.2237>
        <1.1006 6.9519 42.2832>
        <-0.4498 9.5065 40.3055>
        <2.9389 9.1628 40.5715>
        <1.6506 9.9467 39.9646>
        <1.5842 10.4044 39.6103>
        <-0.8873 9.3776 40.4053>
        <-0.4173 11.2766 38.9351>
        <-0.0804 12.4025 38.0634>
        <-3.1163 13.0605 37.554>
        <-0.8523 11.7663 38.5559>
        <-1.8212 13.3587 37.3232>
        <-3.1656 10.9248 39.2074>
        <-3.4974 8.4549 41.1196>
        <-0.3973 13.5028 37.2116>
        <-2.1979 3.1752 51.6213>
        <2.2456 5.0428 49.5808>
        <3.0753 2.2073 52.679>
        <-0.6589 2.6797 52.1628>
        <-0.2779 3.0538 51.754>
        <2.6045 4.6341 50.0273>
        <-1.4347 4.2222 50.4773>
        <-3.3812 3.81 50.9277>
        <-1.3536 5.2221 49.3848>
        <-2.4387 4.4507 50.2276>
        <0.602 6.0482 48.4821>
        <-3.4891 3.6717 51.0789>
        <2.5166 -10.0432 33.481>
        <0.9375 -8.7072 34.496>
        <-1.9094 -11.2736 32.5463>
        <2.1655 -8.7222 34.4846>
        <3.3584 -6.7591 35.9759>
        <-1.0004 -10.5078 33.1281>
        <-1.6958 -9.183 34.1345>
        <-3.7554 -8.809 34.4186>
        <-0.5595 -8.7577 34.4576>
        <-2.6613 -6.7784 35.9613>
        <-2.5186 -6.0304 36.5295>
        <-3.014 -10.0095 33.5066>
        <-3.6013 -5.0091 44.9312>
        <-2.8166 -5.4443 44.6501>
        <-3.5192 -0.9907 47.5269>
        <-3.2665 -2.4498 46.5844>
        <-1.9643 -1.7449 47.0397>
        <-0.1906 -4.3511 45.3563>
        <-2.3735 -1.5794 47.1467>
        <0.1151 -1.7151 47.059>
        <2.1191 -0.3593 47.9348>
        <2.2645 -3.3379 46.0107>
        <-0.0802 -2.7991 46.3588>
        <-0.5498 -1.4489 47.2309>
        <2.3595 -2.2536 46.7111>
        <-1.9551 0.2229 48.5385>
        <1.6833 1.1251 50.9606>
        <0.9834 0.7809 50.0365>
        <2.1508 0.5742 49.4816>
        <2.26 1.7534 52.6476>
        <0.8524 1.2451 51.2828>
        <-2.937 1.2812 51.3799>
        <-1.1092 1.3045 51.4423>
       ]}}
     Random Seed=0
    )
    Socket Reader=$System.Socket Reader
    Socket Sampling Mode=::NewEnumerator3
    VelocitySpeedLimit=false
   )
   &AddVelocity(Velocity=&Multiply_Vector(A=<0 0 7.5>, B=<0 0 10>))
   &CameraOffset(Camera Offset Amount=1)
  ]
  particleUpdateScript [
   &ParticleState()
   &DynamicMaterialParameters(
    Index 0 Param 2=1
    Index 0 Param 3=1
    Module.Index0Float0Write=true
    Module.Index0Float1Write=true
    Module.Index0Float2Write=true
    Module.Index0Float3Write=true
   )
   &Drag(Drag=&RandomRangeFloat(Minimum=5, Maximum=8), Use Linear Drag=true, Use Rotational Drag=true)
   &ScaleMeshSize(Scale Factor=&VectorFromFloat(Value=&FloatFromCurve(FloatCurve=DataInterface() {propertyValues {Curve {keys [
          {arriveTangent=0.3821, interpMode=RCIM_Cubic, leaveTangent=3.4208, tangentMode=RCTM_Break, time=0, value=0}
          {arriveTangentWeight=0.1606, interpMode=RCIM_Cubic, leaveTangentWeight=0.1355, tangentMode=RCTM_Break, time=0.45, value=1}
          {arriveTangentWeight=0.1503, interpMode=RCIM_Cubic, leaveTangentWeight=0.0859, tangentMode=RCTM_User, time=0.7, value=1}
          {arriveTangent=-0.2843, tangentMode=RCTM_User, time=1, value=0}
         ]}}})))
   &CurlNoiseForce(
    Bias Noise=false
    Noise Frequency=32
    Noise Strength=&RandomRangeFloat(Minimum=247.5, Maximum=337.5)
    Pan Noise=true
    Pan Noise Field=<0 0 1>
   )
   &AccelerationForce(Acceleration=<0 0 150>)
   &ScaleColor(
    Scale Alpha=&FloatFromCurve(FloatCurve=DataInterface() {propertyValues {Curve {keys [{interpMode=RCIM_Cubic, tangentMode=RCTM_Break, time=0.75, value=1}, <at 1 0>]}}})
    Scale RGB=&Multiply_VectorByFloat(
     Vector=&MakeVectorFromLinearColorRGB(Color=&ColorFromCurve(ColorCurve=DataInterface() {propertyValues {
         AlphaCurve {keys [<at 0 1 RCIM_Cubic>,<at 1 0 RCIM_Cubic>]}
         BlueCurve {keys [<at 0.2824 0 RCIM_Cubic>,<at 1 0 RCIM_Cubic>]}
         GreenCurve {keys [<at 0.2824 0 RCIM_Cubic>,<at 1 0.0667 RCIM_Cubic>]}
         RedCurve {keys [<at 0.2824 1 RCIM_Cubic>,<at 1 1 RCIM_Cubic>]}
        }}))
     Float=&FNTimeOfDayScalar(Night Scale=0.8, Morning Scale=0.85, Evening Scale=0.9)
    )
    ScaleA=true
    ScaleRGB=true
   )
   /SetVariables(Particles.GlowColor=&MultiplyLinearColor(
     Linear Color=&ColorFromCurve(ColorCurve=DataInterface() {propertyValues {
        AlphaCurve {keys [<at 0.0011 0>, {arriveTangent=0.6438, interpMode=RCIM_Cubic, leaveTangent=0.6438, time=0.15, value=0.6}, <at 0.7 0.45>, <at 1 0 RCIM_Cubic>]}
        BlueCurve {keys [<at 0 0 RCIM_Cubic>,<at 1 0 RCIM_Cubic>]}
        GreenCurve {keys [<at 0 0.1067 RCIM_Cubic>,<at 1 0.035 RCIM_Cubic>]}
        RedCurve {keys [<at 0 0.8 RCIM_Cubic>,<at 1 0.35 RCIM_Cubic>]}
       }})
     Scale Factor=&MakeVector4FromVectorAndFloat("Vector (RGB)"=<1 1 1>, "Float (Alpha)"=&FN_FrontEndFloatModifier(Gameplay=1, Front End=0.5))
    )) {isSetParameters=true}
   &SolveForcesAndVelocity(Clamp Velocity=false, Limit Acceleration=false)
   /SetVariables(Particles.CustomPos:NiagaraPosition=&self:CustomPosition(Position:NiagaraPosition=$Particles.Position, GridSize:NiagaraFloat=1)) {isSetParameters=true}
  ]
  rendererClasses [&NiagaraSpriteRendererProperties, &NiagaraMeshRendererProperties]
  renderers [
   {rendererClass=&NiagaraSpriteRendererProperties}
   {rendererClass=&NiagaraMeshRendererProperties, rendererIndex=1}
  ]
 }
}
summary {
 systemName=NS_Pickaxe_BlockStack_Idle
 emitters [
  {emitterName=Cubes, rendererClasses [&NiagaraSpriteRendererProperties, &NiagaraMeshRendererProperties]}
  {emitterName=CapsuleGlow, rendererClasses [&NiagaraSpriteRendererProperties, &NiagaraMeshRendererProperties]}
 ]
 userVariables [
  User.bIsFrontEnd {type=NiagaraBool}
  User.bIsFrontEndPreview {type=NiagaraBool}
  User.bIsSwinging {type=NiagaraBool}
 ]
}
