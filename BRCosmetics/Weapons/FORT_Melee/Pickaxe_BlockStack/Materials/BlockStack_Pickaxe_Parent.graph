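# Parent material for the BlockStack pickaxe, adding animated fire between the 2023 character parent PreFX and PostFX functions.
# A core flame texture is projected on a pivot-relative virtual plane, giving a flame independent of mesh UVs, with a distance-based mask.
# Two panning noise samples form warp noise that distorts flame textures.
# Tinted core and surface emissive feed an animated fire layer, blended in by a Tech Art Mask channel.
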Material /BRCosmetics/Weapons/FORT_Melee/Pickaxe_BlockStack/Materials/BlockStack_Pickaxe_Parent
graph {
 properties {
  bAutomaticallySetUsageInEditor=false
  bHasPixelAnimation=true
  bUseMaterialAttributes=true
  bUsedWithMeshDeformer=true
  bUsedWithMorphTargets=true
  bUsedWithSkeletalMesh=true
  pixelDepthOffsetMode=PDOM_Legacy
  previewMesh=/BRCosmetics/Weapons/FORT_Melee/Pickaxe_BlockStack/Meshes/BlockStack_Axe
  shadingModel=MSM_FromMaterialExpression
  usedShadingModels=Default Lit | Subsurface | Clear Coat | Subsurface Profile | Cloth
 }
 nodes [
  n1 = MaterialFunctionCall() {materialFunction=&MF_CharacterParent_2023_PreFX}
  MaterialFunctionCall(Material=NamedRerouteUsage(), M=@n1.M, Pre Skinned Local Position=@n1.M, Distance Blend=@n1.M, Fresnel=@n1.M) {materialFunction=&MF_CharacterParent_2023_PostFX}
  NamedRerouteDeclaration(None=@n1.Result) {name=Material In, nodeColor=<rgba 0 0.2734 1 1>}
  n2 = VectorParameter() {group=BlockStack Flame|Core Flame, parameterName=Virtual Plane Offset, defaultValue=<rgba 0.5 0 32 1>}
  n3 = MaterialFunctionCall(
   Plane Normal Axis=Transform(None=Constant3Vector() {constant=<rgba 0 1 0 0>}) {transformSourceType=TRANSFORMSOURCE_Local}
   Plane X Axis=Transform(None=Constant3Vector() {constant=<rgba 1 0 0 0>}) {transformSourceType=TRANSFORMSOURCE_Local}
   Plane Center=Add(
    A=MaterialFunctionCall() {materialFunction=&ObjectPivotPoint}.Object Pivot Location
    B=Transform(None=@n2.RGB) {transformSourceType=TRANSFORMSOURCE_Local}
   )
   UV Scale=Multiply(A=ScalarParameter() {defaultValue=33, group=BlockStack Flame|Core Flame, parameterName=Virtual Plane Scale}) {constB=-1}
  ) {materialFunction=&VirtualPlaneCoordinates}
  NamedRerouteDeclaration(None=@n3.World Position - Centered) {name=Virtual Plane WS Pos, nodeColor=<rgba 0 0.8828 1 1>}
  n4 = Multiply(A=NamedRerouteUsage()) {constB=0.05}
  n5 = Constant()
  NamedRerouteDeclaration(None=Step(Y=Distance(A=@n4, B=@n5))) {name=Virtual Plane Mask, nodeColor=<rgba 0 1 0.1562 1>}
  n6 = Multiply(A=NamedRerouteUsage(), B=Constant2Vector() {g=1, r=-1})
  n7 = Constant2Vector() {r=-1}
  n8 = NamedRerouteUsage()
  n9 = ScalarParameter() {defaultValue=0.02, group=BlockStack Flame|Core Flame, parameterName=Core Flame Warp Intensity}
  n10 = TextureSample(UVs=Add(A=Subtract(A=@n6, B=@n7), B=Multiply(A=@n8, B=@n9))) {samplerSource=SSM_Clamp_WorldGroupSettings, texture=/BRCosmetics/Weapons/FORT_Melee/Pickaxe_BlockStack/Textures/T_BlockStack_Pickaxe_E_Core}
  NamedRerouteDeclaration(None=@n10.RGB) {name=Core Flame, nodeColor=<rgba 1 0 0.1641 1>}
  n11 = NamedRerouteUsage()
  n12 = ScalarParameter() {defaultValue=0.01, group=BlockStack Flame|Emissive, parameterName=Emissive Warp Noise Intensity}
  n13 = TextureSample(UVs=Add(A=TextureCoordinate(), B=Multiply(A=@n11, B=@n12))) {texture=/BRCosmetics/Weapons/FORT_Melee/Pickaxe_BlockStack/Textures/T_BlockStack_Pickaxe_E_TA}
  n14 = Constant3Vector() {constant=<rgba 0.4564 0.0091 0 1>}
  n15 = Step(Y=Desaturation(None=@n13.RGB, Fraction=Constant() {r=1})) {constX=0.1}
  n16 = VectorParameter() {group=BlockStack Flame|Emissive, parameterName=Emissive Color Tint, defaultValue=<rgba 0.9796 1 0.5304 1>}
  n17 = Multiply(
   A=Multiply(A=LinearInterpolate(A=@n13.RGB, B=@n14, Alpha=@n15), B=@n16.RGB)
   B=ScalarParameter() {defaultValue=1.55, group=BlockStack Flame|Emissive, parameterName=Emissive Intensity}
  )
  n18 = NamedRerouteUsage()
  NamedRerouteDeclaration(None=Add(A=NamedRerouteUsage(), B=Add(A=@n17, B=@n18))) {name=Emissive Out, nodeColor=<rgba 0 0.4844 1 1>}
  NamedRerouteDeclaration(None=BlendMaterialAttributes(A=NamedRerouteUsage(), B=NamedRerouteUsage(), Alpha=NamedRerouteUsage())) {name=Material Out, nodeColor=<rgba 1 0 0.9609 1>}
  NamedRerouteDeclaration(None=GetMaterialAttributes(None=NamedRerouteUsage()) {attributeGetTypes [#1, #2]}.Emissive Color) {name=Emissive In, nodeColor=<rgba 0 1 0.7188 1>}
  NamedRerouteDeclaration(None=@n1.Tech Art Mask) {name=TechArtMask In, nodeColor=<rgba 0.8047 1 0 1>}
  n19 = ChannelMaskParameter(None=NamedRerouteUsage()) {group=BlockStack Flame|Emissive, maskChannel=Green, parameterName=FlameMask, defaultValue=<rgba 0 1 0 0>}
  NamedRerouteDeclaration(None=@n19.R) {name=FlameMask, nodeColor=<rgba 0.875 0 1 1>}
  NamedRerouteDeclaration(None=SetMaterialAttributes!(
    MaterialAttributes=NamedRerouteUsage()
    Emissive Color=NamedRerouteUsage()
    Roughness=ScalarParameter() {defaultValue=0.5, group=BlockStack Flame|Emissive, parameterName=Roughness}
    Specular=ScalarParameter() {defaultValue=0.5, group=BlockStack Flame|Emissive, parameterName=Specular}
   ) {attributeSetTypes [#1, #2, #3]}) {name=Animated Fire, nodeColor=<rgba 1 0 0.3984 1>}
  NamedRerouteDeclaration(None=@n3.UVs) {name=Virtual Plane UVs, nodeColor=<rgba 1 0.7969 0 1>}
  n20 = Multiply(A=NamedRerouteUsage(), B=NamedRerouteUsage())
  n21 = VectorParameter() {group=BlockStack Flame|Core Flame, parameterName=Core Flame Color Tint, defaultValue=<rgba 1.55 1.4654 1.1625 1>}
  NamedRerouteDeclaration(None=Multiply(
    A=Multiply(A=@n20, B=@n21.RGB)
    B=ScalarParameter() {defaultValue=1.55, group=BlockStack Flame|Core Flame, parameterName=Core Flame Intensity}
   )) {name=Core Flame Out, nodeColor=<rgba 1 0.0633 0 1>}
  n22 = NamedRerouteUsage()
  n23 = ScalarParameter() {defaultValue=0.3, group=BlockStack Flame|Warp Noise, parameterName=Noise 1 Scale}
  n24 = TextureSample(UVs=Panner(Coordinate=Multiply(A=@n22, B=@n23)) {speedX=0.1, speedY=0.3}) {samplerType=SAMPLERTYPE_LinearGrayscale, texture=/Game/Tools/WaterfallGenerator/Art/T_Waterfall_NoiseA}
  n25 = ScalarParameter() {defaultValue=0.5, group=BlockStack Flame|Warp Noise, parameterName=Noise 2 Scale}
  n26 = TextureSample(UVs=Panner(Coordinate=Multiply(A=@n22, B=@n25)) {speedX=0.15, speedY=0.6}) {samplerType=SAMPLERTYPE_LinearGrayscale, texture=&T_Waterfall_NoiseA}
  NamedRerouteDeclaration(None=Add(A=Subtract(A=@n24.R) {constB=0.5}, B=Subtract(A=@n26.R) {constB=0.5})) {name=Warp Noise, nodeColor=<rgba 0.4766 0 1 1>}
 ]
}
