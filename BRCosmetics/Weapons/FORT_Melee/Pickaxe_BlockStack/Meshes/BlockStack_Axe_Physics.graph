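# Physics asset for the BlockStack pickaxe skeletal mesh, giving its two flame pieces physical bodies linked to the weapon handle.
# A capsule body covers the two-handed handle bone, and small capsules cover the left and right dyn_flame bones.
# Each flame bone is constrained to the handle with tight swing and twist limits of a few degrees, allowing only slight secondary motion.
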PhysicsAsset /BRCosmetics/Weapons/FORT_Melee/Pickaxe_BlockStack/Meshes/BlockStack_Axe_Physics
bodies {
 Melee_TwoHanded_Handle [{shapeType=Capsule, length=14.7425, radius=4.312, center=<-0.3198 -0.0045 0.4247>, rotation=<rot 0.6012 -90 0.0002>}]
 dyn_flame_l [{shapeType=Capsule, length=5, radius=2, shapeName=dyn_flame_l_capsule, center=<3 0 0>, rotation=<rot 90 0 0>}]
 dyn_flame_r [{shapeType=Capsule, length=5, radius=2, shapeName=dyn_flame_r_capsule, center=<3 0 0>, rotation=<rot 90 0 0>}]
}
constraints [
 {bone1Name=dyn_flame_r, bone2Name=Melee_TwoHanded_Handle, swing1LimitDegrees=5, swing2LimitDegrees=1, twistLimitDegrees=1}
 {bone1Name=dyn_flame_l, bone2Name=Melee_TwoHanded_Handle, swing1LimitDegrees=5, swing2LimitDegrees=1, twistLimitDegrees=1}
]
