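# Anim Blueprint for the BlockStack pickaxe that adds secondary motion to its flame bones.
# The cached input pose feeds a bool blend between a RigidBody-simulated branch with upward gravity and an unsimulated one, both ending in additive left/right flame bone transforms.
# Each update computes sine-of-time ember vectors, though no visible node reads them.
# The Anim Dynamics nodes and two other RigidBody setups do not reach the output.
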AnimBlueprint /BRCosmetics/Weapons/FORT_Melee/Pickaxe_BlockStack/Meshes/BlockStack_Axe_AnimBP
target_skeleton=&/BlockStack_Axe_Skeleton
graphs [
 AnimGraph {nodes [
   Anim Dynamics - Bone: dyn_flame_c_01(Alpha=1, ExternalForce="0,0,0") {
    angularDampingOverride=1
    angularSpringConstant=20000
    bAngularSpring=true
    bOverrideAngularDamping=true
    bUseGravityOverride=true
    gravityScale=0.35
    boundBone {boneName=dyn_flame_c_01}
    componentLinearAccScale=<1 1 1>
    componentLinearVelScale=<1 1 1>
    gravityOverride {z=100}
    physicsBodyDefinitions [{
      boundBone {boneName=dyn_flame_c_01}
      boxExtents=<30 40 40>
      constraintSetup {angularLimitsMax=<15 15 15>, angularLimitsMin=<-15 -15 -15>, angularTarget=<1 0 0>}
      localJointOffset=<4 0 0>
     }]
   }
   n1 = Anim Dynamics - Bone: base_flame(Alpha=1, ExternalForce="0,0,0") {
    angularDampingOverride=1
    angularSpringConstant=20000
    bAngularSpring=true
    bOverrideAngularDamping=true
    bUseGravityOverride=true
    gravityScale=0.35
    boundBone {boneName=base_flame}
    componentLinearAccScale=<1 1 1>
    componentLinearVelScale=<1 1 1>
    gravityOverride {z=100}
    physicsBodyDefinitions [{
      boundBone {boneName=base_flame}
      boxExtents=<2 2 2>
      constraintSetup {angularLimitsMax=<0 0 0>, angularLimitsMin=<0 0 0>, angularTarget=<0 0 0>}
      localJointOffset=<0 0 0>
     }]
   }
   n2 = Anim Dynamics - Bone: dyn_flame_c_01(ComponentPose=@n1, Alpha=1, ExternalForce="0,0,0") {
    angularDampingOverride=1
    angularSpringConstant=20000
    bOverrideAngularDamping=true
    bUseGravityOverride=true
    gravityScale=0.35
    boundBone {boneName=dyn_flame_c_01}
    componentLinearAccScale=<1 1 1>
    componentLinearVelScale=<1 1 1>
    gravityOverride {z=100}
    physicsBodyDefinitions [{
      boundBone {boneName=dyn_flame_c_01}
      boxExtents=<2 2 2>
      constraintSetup {angularLimitsMax=<0 0 0>, angularLimitsMin=<0 0 0>, angularTarget=<1 0 0>}
      localJointOffset=<0 0 0>
     }]
   }
   n3 = Anim Dynamics - Bone: dyn_flame_c_02(ComponentPose=@n2, Alpha=1, ExternalForce="0,0,0") {
    angularDampingOverride=1
    angularSpringConstant=20000
    bAngularSpring=true
    bOverrideAngularDamping=true
    bUseGravityOverride=true
    gravityScale=0.35
    boundBone {boneName=dyn_flame_c_02}
    componentLinearAccScale=<1 1 1>
    componentLinearVelScale=<1 1 1>
    gravityOverride {z=100}
    physicsBodyDefinitions [{
      boundBone {boneName=dyn_flame_c_02}
      boxExtents=<2 2 2>
      constraintSetup {angularLimitsMax=<0 0 0>, angularLimitsMin=<0 0 0>, angularTarget=<1 0 0>}
      localJointOffset=<0 0 0>
     }]
   }
   n4 = Anim Dynamics - Bone: dyn_flame_c_03(ComponentPose=@n3, Alpha=1, ExternalForce="0,0,0") {
    angularDampingOverride=1
    angularSpringConstant=20000
    bAngularSpring=true
    bOverrideAngularDamping=true
    bUseGravityOverride=true
    gravityScale=0.35
    boundBone {boneName=dyn_flame_c_03}
    componentLinearAccScale=<1 1 1>
    componentLinearVelScale=<1 1 1>
    gravityOverride {z=100}
    physicsBodyDefinitions [{
      boundBone {boneName=dyn_flame_c_03}
      boxExtents=<2 2 2>
      constraintSetup {angularLimitsMax=<0 0 0>, angularLimitsMin=<0 0 0>, angularTarget=<1 0 0>}
      localJointOffset=<0 0 0>
     }]
   }
   n5 = Anim Dynamics - Bone: dyn_flame_c_04(ComponentPose=@n4, Alpha=1, ExternalForce="0,0,0") {
    angularDampingOverride=1
    angularSpringConstant=20000
    bAngularSpring=true
    bOverrideAngularDamping=true
    bUseGravityOverride=true
    gravityScale=0.35
    boundBone {boneName=dyn_flame_c_04}
    componentLinearAccScale=<1 1 1>
    componentLinearVelScale=<1 1 1>
    gravityOverride {z=100}
    physicsBodyDefinitions [{
      boundBone {boneName=dyn_flame_c_04}
      boxExtents=<2 2 2>
      constraintSetup {angularLimitsMax=<0 0 0>, angularLimitsMin=<0 0 0>, angularTarget=<1 0 0>}
      localJointOffset=<0 0 0>
     }]
   }
   n6 = Anim Dynamics - Bone: dyn_flame_r(ComponentPose=@n5, Alpha=1, ExternalForce="0,0,0") {
    angularDampingOverride=1
    angularSpringConstant=20000
    bAngularSpring=true
    bOverrideAngularDamping=true
    bUseGravityOverride=true
    gravityScale=0.35
    boundBone {boneName=dyn_flame_r}
    componentLinearAccScale=<1 1 1>
    componentLinearVelScale=<1 1 1>
    gravityOverride {z=100}
    physicsBodyDefinitions [{
      boundBone {boneName=dyn_flame_r}
      boxExtents=<30 40 40>
      constraintSetup {angularLimitsMax=<15 15 15>, angularLimitsMin=<-15 -15 -15>, angularTarget=<1 0 0>, linearAxesMax=<2 2 4>}
      localJointOffset=<4 0 0>
     }]
   }
   Anim Dynamics - Bone: dyn_flame_l(ComponentPose=@n6, Alpha=1, ExternalForce="0,0,0") {
    angularDampingOverride=1
    angularSpringConstant=20000
    bAngularSpring=true
    bOverrideAngularDamping=true
    bUseGravityOverride=true
    gravityScale=0.35
    boundBone {boneName=dyn_flame_l}
    componentLinearAccScale=<1 1 1>
    componentLinearVelScale=<1 1 1>
    gravityOverride {z=100}
    physicsBodyDefinitions [{
      boundBone {boneName=dyn_flame_l}
      boxExtents=<30 40 40>
      constraintSetup {angularLimitsMax=<15 15 15>, angularLimitsMin=<-15 -15 -15>, angularTarget=<1 0 0>, linearAxesMax=<2 2 2>}
      localJointOffset=<4 0 0>
     }]
   }
   Anim Dynamics - Bone: base_flame(Alpha=1, ExternalForce="0,0,0") {
    angularDampingOverride=1
    angularSpringConstant=20000
    bAngularSpring=true
    bOverrideAngularDamping=true
    bUseGravityOverride=true
    gravityScale=0.35
    boundBone {boneName=base_flame}
    componentLinearAccScale=<1 1 1>
    componentLinearVelScale=<1 1 1>
    gravityOverride {z=100}
    physicsBodyDefinitions [{
      boundBone {boneName=base_flame}
      boxExtents=<30 40 40>
      constraintSetup {angularLimitsMax=<15 15 15>, angularLimitsMin=<-15 -15 -15>, angularTarget=<1 0 0>}
      localJointOffset=<4 0 0>
     }]
   }
   n7 = RigidBody() {
    bClampLinearTranslationLimitToRefPose=true
    bOverrideWorldGravity=true
    overridePhysicsAsset=&/BlockStack_Axe_Physics
    simulationSpace=BaseBoneSpace
    alphaBoolBlend {blendOption=Linear}
    baseBoneRef {boneName=b_AR_Root}
    componentAppliedLinearAccClamp=<0 0 0>
    componentLinearVelScale=<25 25 25>
    overrideWorldGravity {z=100}
    simSpaceSettings {
     maxAngularAcceleration=80
     maxAngularVelocity=2
     maxLinearAcceleration=3500
     maxLinearVelocity=700
     worldAlpha=1
     externalLinearDragV=<0.2 0.2 -0.8>
    }
   }
   Component To Local(ComponentPose=@n7)
   RigidBody() {
    bClampLinearTranslationLimitToRefPose=true
    bOverrideWorldGravity=true
    overridePhysicsAsset=&BlockStack_Axe_Physics
    simulationSpace=BaseBoneSpace
    alphaBoolBlend {blendOption=Linear}
    baseBoneRef {boneName=b_AR_Root}
    componentAppliedLinearAccClamp=<0 0 0>
    overrideWorldGravity {z=100}
    simSpaceSettings {
     dampingAlpha=0
     maxAngularAcceleration=40
     maxAngularVelocity=10
     maxLinearAcceleration=3500
     maxLinearVelocity=700
     worldAlpha=0.01
     externalLinearDragV=<0.5 0.5 0.5>
    }
   }
   n8 = RigidBody(ComponentPose=Local To Component(LocalPose=Use cached pose 'InputPoseCache'())) {
    bClampLinearTranslationLimitToRefPose=true
    bOverrideWorldGravity=true
    overridePhysicsAsset=&BlockStack_Axe_Physics
    simulationSpace=BaseBoneSpace
    alphaBoolBlend {blendOption=Linear}
    baseBoneRef {boneName=b_AR_Root}
    componentAppliedLinearAccClamp=<0 0 0>
    overrideWorldGravity {z=100}
    simSpaceSettings {
     maxAngularAcceleration=40
     maxAngularVelocity=10
     maxLinearAcceleration=3500
     maxLinearVelocity=700
     worldAlpha=1
     externalLinearDragV {y=0.5, z=0.5}
    }
   }
   n9 = "Transform (Modify) Bone - Bone: dyn_flame_l"(
    ComponentPose="Transform (Modify) Bone - Bone: dyn_flame_r"(ComponentPose=@n8, Alpha=1, Translation="0,0,0", Rotation="0,0,0", Scale="1,1,1") {translationMode=BMM_Additive, translationSpace=BCS_BoneSpace, boneToModify {boneName=dyn_flame_r}}
    Alpha=1
    Translation="0,0,0"
    Rotation="0,0,0"
    Scale="1,1,1"
   ) {translationMode=BMM_Additive, translationSpace=BCS_BoneSpace, boneToModify {boneName=dyn_flame_l}}
   n10 = Use cached pose 'InputPoseCache'()
   n11 = "Transform (Modify) Bone - Bone: dyn_flame_l"(
    ComponentPose="Transform (Modify) Bone - Bone: dyn_flame_r"(ComponentPose=Local To Component(LocalPose=@n10), Alpha=1, Translation="0,0,0", Rotation="0,0,0", Scale="1,1,1") {translationMode=BMM_Additive, translationSpace=BCS_BoneSpace, boneToModify {boneName=dyn_flame_r}}
    Alpha=1
    Translation="0,0,0"
    Rotation="0,0,0"
    Scale="1,1,1"
   ) {translationMode=BMM_Additive, translationSpace=BCS_BoneSpace, boneToModify {boneName=dyn_flame_l}}
   Output Pose(Result=Blend Poses by bool(BlendPose_0=Component To Local(ComponentPose=@n11), BlendPose_1=Component To Local(ComponentPose=@n9))) {layerGroup=None}
   AnimGraphNode_SubInput_0 = Input Pose()
   Save cached pose 'InputPoseCache'(Pose=@AnimGraphNode_SubInput_0) {cacheName=InputPoseCache}
  ]}
 EventGraph {nodes [
   Event BlueprintUpdateAnimation()
   n1 = float * float(A=GetRealTimeSeconds(), B=4) {is=title K2Node_PromotableOperator}
   n2 = float * float(A=GetRealTimeSeconds(), B=7)
   > Set Ember01(Ember01:Vector=MakeVector(Y="Sin (Radians)"(A=@n2), Z="Sin (Radians)"(A=@n1))) -> (Output_Get:Vector)
   n3 = float * float(A=GetRealTimeSeconds(), B=5)
   n4 = float * float(A=GetRealTimeSeconds(), B=4)
   > Set Ember_L(Ember_L:Vector=MakeVector(Y="Sin (Radians)"(A=@n3), Z="Sin (Radians)"(A=@n4))) -> (Output_Get:Vector)
   TryGetPawnOwner()
  ]}
]
variables [
 Ember01 {type=Vector}
 Ember_L {type=Vector}
]
